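; Interactable yacht flag prop that lets a player "Salute" it for a Donut quest.
; Interacting sends a complex custom stat event tagged Athena.Quests.Donut.SaluteShorts to the interacting player's quest manager.
; The interaction check always passes on authority; on remotes it passes only while the first local player has the Quest_S12_Donut_W9_B quest with that objective incomplete.
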(event EventBeginPlay)

(event Interaction|EventBlueprintOnInteract
  (params (InteractingPawn :type Fort%20Pawn%20Object%20Reference) (InteractionBeingAttempted :type EInteractionBeingAttempted%20Enum :default "FirstInteraction"))
  (Quest|SendComplexCustomStatEvent (Quests|GetQuestManager (Fortnite|GetFortPlayerControllerfromActor InteractingPawn) "Invalid") 0 "(GameplayTags=)" "(GameplayTags=((TagName=\"Athena.Quests.Donut.SaluteShorts\")))" 1))

(fn BlueprintCanInteract
  (params (InteractingPawn :type Fort%20Pawn%20Object%20Reference) (InteractionBeingAttempted :type EInteractionBeingAttempted%20Enum :default "FirstInteraction") (InteractionType :type TInteractionType%20Enum :default "IT_NoInteraction"))
  (bind _returnvalue (Fortnite|GetFirstLocalFortPlayerController true))
  (bind _donutquest (Variables|Default|GetDonutQuest))
  (Networking|SwitchHasAuthority
    (:Authority
      (return true))
    (:Remote
      (bind _questvalid (Quest|DoesPlayerHaveQuest _returnvalue _donutquest "Quest_S12_Donut_W9_B"))
      (bind (_playercontrollerout _completedquestobjective) (Quest|HasPlayerCompletedQuestObjective _returnvalue _donutquest "Quest_S12_Donut_W9_B"))
      (return (and _questvalid (not _completedquestobjective))))))

(fn BlueprintGetInteractionString
  (params (InteractingPawn :type Fort%20Pawn%20Object%20Reference) (InteractionBeingAttempted :type EInteractionBeingAttempted%20Enum :default "FirstInteraction"))
  (return "NSLOCTEXT(\"[AFC00CB34394D821E893D38EFACADEDD]\", \"585BF88B4BC4EAC8E94BC08DB64F695D\", \"Salute\")"))

(fn ConstructionScript)
